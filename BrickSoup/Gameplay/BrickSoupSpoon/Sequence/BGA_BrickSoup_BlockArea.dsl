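; Brick Soup block-area actor: a sphere zone that announces its spawn and applies a blocking gameplay effect to human players inside.
; On authority, it adds a delayed minimap radius marker and optionally blocks vehicles via a scalable-float hotfix.
; Entering also strips the in-vehicle effect to force ejection; leaving removes the block effect.
; Sphere-overlap queries feed functions for knockback, camera-lens cues and granting or removing external emotes.
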(event EventBeginPlay
  (GameplayEventRouter|BroadcastStatefulEventBackwardCompatible (GameplayEventRouter|GetGlobalEventRouter) "(TagName=\"EventMessage.Athena.BrickSoup.BlockAreaSpawned\")" (Utilities|Struct|MakeBrickSoupObjectiveBlockArea self))
  ; Setting up Minimap Radial Icon
  (Networking|SwitchHasAuthority
    (:Authority
      (Utilities|FlowControl|Delay 1e+01)
      (bind _returnvalue (RadiusTracker|MakeRadiusTrackingInitializer 0.0 (Variables|/Script/Engine.SphereComponent|GetSphereRadius (Variables|Default|GetSphere)) (Variables|/Script/Engine.SphereComponent|GetSphereRadius (Variables|Default|GetSphere)) 2.0 0.0 "(R=1.000000,G=0.072917,B=0.072917,A=1.000000)" "/BrickSoup/UI/Materials/M_UI_BrickSoup_MinimapBlock.M_UI_BrickSoup_MinimapBlock" self (Fortnite|GetAllFortPlayerControllers true) "Global" 1e+07))
      (Fortnite|AddActorTrackingInfo (RadiusTracker|GetGameStateRadiusTrackerComponent) _returnvalue)
      ; Adding Hotfix to block vehicles from entering the area
      (if (ScalableFloat|GetValueasBool (Variables|Default|GetRow_ShouldBlockVehicles))
        (Collision|SetCollisionResponsetoChannel (Variables|Default|GetSphere) "ECC_Vehicle" "ECR_Block")))))

(event Collision|EventActorBeginOverlap
  (params (OtherActor :type Actor%20Object%20Reference))
  (bind _asfort_player_pawn (Utilities|Casting|CastToFortPlayerPawn OtherActor))
  (if (Pawn|IsBotControlled _asfort_player_pawn)
    _)
  (GameplayEffects|ApplyGameplayEffectToTarget :self (Ability|GetAbilitySystemComponent self) :GameplayEffectClass (Variables|Default|GetGE_Block) :Target (Ability|GetAbilitySystemComponent _asfort_player_pawn) :Level 0.0)
  ; Force ejecting players from vehicles
  (GameplayEffects|RemoveActiveEffectswithGrantedTags (Ability|GetAbilitySystemComponent _asfort_player_pawn) "(GameplayTags=((TagName=\"Gameplay.Action.Player.AthenaInVehicle\")))"))

(event OnComponentBeginOverlap(Sphere)
  (params (OverlappedComponent :type Primitive%20Component%20Object%20Reference) (OtherActor :type Actor%20Object%20Reference) (OtherComp :type Primitive%20Component%20Object%20Reference) (OtherBodyIndex :type Integer) (bFromSweep :type Boolean) (SweepResult :type Hit%20Result%20Structure%20%28by%20ref%29)))

(event Collision|EventActorEndOverlap
  (params (OtherActor :type Actor%20Object%20Reference))
  (bind _asfort_player_pawn (Utilities|Casting|CastToFortPlayerPawn OtherActor))
  (if (Pawn|IsBotControlled _asfort_player_pawn)
    _)
  (GameplayEffects|RemoveActiveGameplayEffectbySourceEffect (Ability|GetAbilitySystemComponent _asfort_player_pawn) (Variables|Default|GetGE_Block) 0 -1))

(fn ApplyCameraLens
  (bind _outactors (CallFunction|GetPawnsinArea))
  (for _array_element _outactors
    (bind _asfort_player_pawn (Utilities|Casting|CastToFortPlayerPawn _array_element))
    (GameplayCue|ExecuteGameplayCue _asfort_player_pawn (Variables|Default|GetGCN_CameraLens))))

(fn DoKnockback
  (bind _outactors (CallFunction|GetPawnsinArea))
  (for _array_element _outactors
    (GameplayEffects|FortApplyGameplayEffectToTarget :self (Ability|GetFortAbilitySystemComponent self) :GameplayEffectClass (Variables|Default|GetKnockbackEffect) :TargetOfEffect (Ability|GetFortAbilitySystemComponent _array_element) :OptionalSourceObject self :OptionalEffectCauser self :OptionalDamageSource self :Level 0.0)))

(fn GetPawnsInArea
  (bind _outactors (Collision|ComponentOverlapActors (Variables|Default|GetSphere) (Transformation|GetWorldTransform (Variables|Default|GetSphere)) (Variables|Default|GetOverlapObjectTypes) (Variables|Default|GetOverlapClassFilter)))
  (return _outactors))

(fn GrantEmote
  (params (PlayerPawn :type Fort%20Player%20Pawn%20Object%20Reference))
  (bind _returnvalue (Fortnite|GetFortPlayerControllerfromActor PlayerPawn))
  (bind _returnvalue_2 (Actor|GetComponentbyClass _returnvalue "/Script/FortniteGame.FortControllerComponent_ExternalEmote"))
  (Utilities|IsValid _returnvalue_2
    (:Is%20Valid
      (Class|FortControllerComponentExternalEmote|AddExternalMultiEmoteCategories _returnvalue_2 (Utilities|Array|MakeArray (Variables|Default|GetEmoteCategory))))))

(fn GrantPlayersEmote
  (bind _outactors (CallFunction|GetPawnsinArea))
  (for _array_element _outactors
    (bind _asfort_player_pawn (Utilities|Casting|CastToFortPlayerPawn _array_element))
    (Default|GrantEmote _asfort_player_pawn)))

(fn RemoveEmote
  (params (PlayerPawn :type Fort%20Player%20Pawn%20Object%20Reference))
  (bind _returnvalue (Fortnite|GetFortPlayerControllerfromActor PlayerPawn))
  (bind _returnvalue_2 (Actor|GetComponentbyClass _returnvalue "/Script/FortniteGame.FortControllerComponent_ExternalEmote"))
  (Utilities|IsValid _returnvalue_2
    (:Is%20Valid
      (Class|FortControllerComponentExternalEmote|RemoveExternalMultiEmoteCategories _returnvalue_2 (Utilities|Array|MakeArray (Utilities|Struct|BreakExternalMultiEmoteCategory (Variables|Default|GetEmoteCategory)))))))

(fn RemovePlayersEmote
  (bind _outactors (CallFunction|GetPawnsinArea))
  (for _array_element _outactors
    (bind _asfort_player_pawn (Utilities|Casting|CastToFortPlayerPawn _array_element))
    (Default|RemoveEmote _asfort_player_pawn)))

(fn ConstructionScript)
